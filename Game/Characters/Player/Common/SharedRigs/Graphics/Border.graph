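# Translucent UI material in the player shared-rig graphics that procedurally builds a rectangular border mask from UVs, without textures.
# Each U and V axis and its inverse are scaled by Mult, raised to Contrast and saturated, then multiplied and inverted to isolate both edges.
# The horizontal and vertical edge masks are added, clamped and sharpened with CheapContrast.
# A Color parameter defaults to white RGB; output connections are not shown.
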Material /Game/Characters/Player/Common/SharedRigs/Graphics/Border
graph {
 properties {blendMode=BLEND_Translucent, materialDomain=MD_UI, pixelDepthOffsetMode=PDOM_Legacy, usedShadingModels=Default Lit}
 nodes [
  n1 = TextureCoordinate()
  n2 = ComponentMask(None=@n1) {r=true}
  n3 = OneMinus(None=@n2)
  n4 = ScalarParameter() {defaultValue=30, parameterName=Mult}
  n5 = ScalarParameter() {defaultValue=100, parameterName=Contrast}
  n6 = Saturate(None=Power(Base=Multiply(A=@n3, B=@n4) {constB=5}, Exp=@n5) {constExponent=5})
  n7 = Saturate(None=Power(Base=Multiply(A=@n2, B=@n4) {constB=5}, Exp=@n5) {constExponent=5})
  n8 = Saturate(None=OneMinus(None=Multiply(A=@n6, B=@n7)))
  n9 = ComponentMask(None=@n1) {g=true}
  n10 = OneMinus(None=@n9)
  n11 = Saturate(None=Power(Base=Multiply(A=@n10, B=@n4) {constB=5}, Exp=@n5) {constExponent=5})
  n12 = Saturate(None=Power(Base=Multiply(A=@n9, B=@n4) {constB=5}, Exp=@n5) {constExponent=5})
  n13 = Saturate(None=OneMinus(None=Multiply(A=@n11, B=@n12)))
  MaterialFunctionCall(In=Saturate(None=Add(A=@n8, B=@n13)), Contrast=Constant() {r=10}) {materialFunction=&CheapContrast}
  VectorParameter() {parameterName=Color, defaultValue=<rgba 1 1 1 0>}
 ]
}
