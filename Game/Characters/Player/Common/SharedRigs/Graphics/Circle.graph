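# Translucent UI material in the player's shared-rig graphics that computes a filled circle mask procedurally, without textures.
# UVs are remapped around the center and measured with VectorLength, then offset by the circle size parameter.
# A narrow parameter-driven SmoothStep and OneMinus turn that distance into a crisp disc.
# A Color parameter and stray Time/Frac, Rotator, Add and Lerp nodes sit outside this mask chain.
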Material /Game/Characters/Player/Common/SharedRigs/Graphics/Circle
graph {
 properties {blendMode=BLEND_Translucent, materialDomain=MD_UI, pixelDepthOffsetMode=PDOM_Legacy, usedShadingModels=Default Lit}
 nodes [
  n1 = TextureCoordinate()
  Add(A=@n1) {constB=-0.5}
  Frac(None=Time())
  LinearInterpolate() {constAlpha=0}
  n2 = MaterialFunctionCall(Vector 2=ConstantBiasScale(None=@n1) {bias=-0.5, scale=2}) {materialFunction=/Engine/Functions/Engine_MaterialFunctions02/Utility/VectorLength}
  n3 = ScalarParameter() {defaultValue=0.7, parameterName=Circle SIze}
  OneMinus(None=SmoothStep(
    Min=ScalarParameter() {parameterName=0}
    Max=ScalarParameter() {defaultValue=0.001, parameterName=0.01}
    Value=Subtract(A=@n2.V2 Length, B=@n3)
   ))
  Rotator()
  VectorParameter() {parameterName=Color, defaultValue=<rgba 1 1 1 0>}
 ]
}
